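; Geometry Script generated-mesh actor that procedurally scatters moss meshes over a source static mesh and can bake them into a static mesh asset.
; On rebuild, non-uniform point sampling yields candidates, which a masked vertex-color threshold filters.
; Seeded streams vary mesh choice, rotation and scale, with cell-distance noise and edge falloff, and each copy's normals match its surface sample.
; Bake creates or updates a Nanite "_MossAddon" asset using the first moss material.
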(event EventBeginPlay)

(event Collision|EventActorBeginOverlap
  (params (OtherActor :type Actor%20Object%20Reference)))

(event EventTick
  (params (DeltaSeconds :type Float%20%28single-precision%29)))

(event Events|EventOnRebuildGeneratedMesh
  (params (TargetMesh :type Dynamic%20Mesh%20Object%20Reference))
  (bind _vertexcolormask (Variables|Default|GetVertexColorMask))
  (bind (_r _g _b _a) (Utilities|Struct|BreakLinearColor _vertexcolormask))
  (bind _returnvalue (Math|Conversions|ToFloat(Integer) 0))
  (bind _tempbasemesh (Variables|Default|GetTempBaseMesh))
  (bind _tempmoss (Variables|Default|GettempMoss))
  (bind _targetmeshmoss_array (Variables|Default|GetTargetMeshMoss_Array))
  (bind _targetmeshmoss_array_2 (Variables|Default|GetTargetMeshMoss_Array))
  (bind _seed (Variables|Default|GetSeed))
  (bind _cellnoisefrequency (Variables|Default|GetCellNoiseFrequency))
  (Variables|Default|SetTarget%20Mesh TargetMesh)
  (bind _returnvalue_2 (DynamicMeshActor|AllocateComputeMesh))
  (Variables|Default|SetTempBaseMesh _returnvalue_2)
  (GeometryScript|StaticMesh|CopyMeshFromStaticMeshwithSectionMaterials (Variables|Default|GetStaticmesh) (Variables|Default|GetTempBaseMesh) (Utilities|Struct|MakeGeometryScriptCopyMeshfromAssetOptions true) (Utilities|Struct|MakeGeometryScriptMeshReadLOD "MaxAvailable")
    (:Success
      (if (Variables|Default|GetEnabled)
        (Utilities|FlowControl|Sequence
          (:then_0
            (Utilities|Array|Clear (Variables|Default|GetTargetMeshMoss_Array))
            (for _array_element (Variables|Default|GetMossMesh)
              (bind _returnvalue_3 (DynamicMeshActor|AllocateComputeMesh))
              (bind _returnvalue_4 (Utilities|Array|Add _targetmeshmoss_array _returnvalue_3))
              (GeometryScript|StaticMesh|CopyMeshFromStaticMeshwithSectionMaterials _array_element (Utilities|Array|Get(acopy) _targetmeshmoss_array _returnvalue_4) (Utilities|Struct|MakeGeometryScriptCopyMeshfromAssetOptions true) (Utilities|Struct|MakeGeometryScriptMeshReadLOD "MaxAvailable"))))
          (:then_2
            (bind (_samples _sampleradii _triangleids) (GeometryScript|MeshSampling|ComputeNonUniformPointSampling _tempbasemesh (Utilities|Struct|MakeGeometryScriptMeshPointSamplingOptions (Variables|Default|GetSampling%20Radius_Min) (Variables|Default|GetMax%20Num%20Samples) 1) (Utilities|Struct|MakeGeometryScriptNonUniformPointSamplingOptions (Variables|Default|GetSampling%20Radius_Max) (Variables|Default|GetSample%20Size%20Distribution))))
            (Variables|Default|SetRandomStream (Math|Random|MakeRandomStream _seed))
            (Variables|Default|SetRandomStream_2 (Math|Random|MakeRandomStream (+ _seed 342)))
            (for (_array_element_2 _array_index) _samples
              (bind (_bisvalidtriangle _vertex1 _vertex2 _vertex3 _barycentriccoords) (GeometryScript|MeshQueries|ComputeTriangleBarycentricCoords _tempbasemesh (.ReturnValue (GeometryScript|ListUtils|GetIndexListItem _triangleids _array_index)) (.location _array_element_2)))
              (bind (_btrihasvalidvertexcolors _interpolatedcolor) (GeometryScript|MeshQueries|GetInterpolatedTriangleVertexColor _tempbasemesh (.ReturnValue (GeometryScript|ListUtils|GetIndexListItem _triangleids _array_index)) _barycentriccoords))
              (Development|PrintString (Utilities|String|ToString(Float) (Utilities|Operators|Add (* (Utilities|Struct|BreakLinearColor _interpolatedcolor) _r) (* (.G (Utilities|Struct|BreakLinearColor _interpolatedcolor)) _g) (* _b (.B (Utilities|Struct|BreakLinearColor _interpolatedcolor))) (* (.A (Utilities|Struct|BreakLinearColor _interpolatedcolor)) _a))))
              (if (> (Utilities|Operators|Add (* (Utilities|Struct|BreakLinearColor _interpolatedcolor) _r) (* (.G (Utilities|Struct|BreakLinearColor _interpolatedcolor)) _g) (* _b (.B (Utilities|Struct|BreakLinearColor _interpolatedcolor))) (* (.A (Utilities|Struct|BreakLinearColor _interpolatedcolor)) _a)) (Variables|Default|GetVertexColorThreshold))
                (Variables|Default|SetCurrentRandomVal (Math|Random|RandomIntegerinRangefromStream (Variables|Default|GetRandomStream_2) 0 (Utilities|Array|LastIndex _targetmeshmoss_array_2)))
                (bind _returnvalue_5 (DynamicMeshActor|AllocateComputeMesh))
                (bind _output_get (Variables|Default|SettempMoss _returnvalue_5))
                (bind _copytomeshout (GeometryScript|Decomposition|CopyMeshtoMesh (Utilities|Array|Get(acopy) _targetmeshmoss_array_2 (Variables|Default|GetCurrentRandomVal)) _output_get))
                (GeometryScript|Transforms|TransformMesh _tempmoss (Math|Transform|MakeTransform (+ (.location _array_element_2) (* (Math|Vector|Normalize (Math|Vector|GetUpVector (.rotation _array_element_2))) (+ (Math|Float|Lerp (Variables|Default|GetEdgeZoffset) 0.0 (Math|Float|Clamp(Float) (Utilities|Operators|Add (* (Utilities|Struct|BreakLinearColor _interpolatedcolor) _r) (* (.G (Utilities|Struct|BreakLinearColor _interpolatedcolor)) _g) (* _b (.B (Utilities|Struct|BreakLinearColor _interpolatedcolor))) (* (.A (Utilities|Struct|BreakLinearColor _interpolatedcolor)) _a)))) (Variables|Default|GetZoffset)))) (Math|Rotator|CombineRotators (Math|Rotator|CombineRotators (.rotation _array_element_2) (Math|Vector|RotatorfromAxisandAngle (Math|Vector|GetUpVector (.rotation _array_element_2)) (Math|Random|RandomFloatinRangefromStream (Variables|Default|GetRandomStream) 0.0 3.6e+02))) (Math|Vector|RotatorfromAxisandAngle (Math|Vector|CrossProduct (Math|Vector|GetUpVector (.rotation _array_element_2)) "0.000000,0.000000,1.000000") (Variables|Default|GetAglineToSurface))) (Math|Vector|MakeVector (Utilities|Operators|Multiply (Math|Random|RandomFloatinRangefromStream (Variables|Default|GetRandomStream) (Variables|Default|GetScaleMin) (Variables|Default|GetScaleMax)) (Math|Float|Lerp (Variables|Default|GetNoiseScale_Max) (Variables|Default|GetNoiseScale_Min) (Math|Float|Power (Math|Float|Clamp(Float) (/ (Math|Vector|Distance(Vector) (.location _array_element_2) (+ (Math|Vector|MakeVector _returnvalue (Math|Conversions|ToFloat(Integer) (Math|Float|Floor (.y (/ (.location _array_element_2) _cellnoisefrequency)))) (Math|Conversions|ToFloat(Integer) (Math|Float|Floor (.z (/ (.location _array_element_2) _cellnoisefrequency))))) "0.500000,0.500000,0.500000")) _cellnoisefrequency)) (Variables|Default|GetNoise_FallOff))) (Math|Float|Lerp 1.0 (/ (Utilities|Array|Get(acopy) _sampleradii _array_index) (Variables|Default|GetSampling%20Radius_Min)) (Variables|Default|GetScale_BySampleRadiusBlend)) (Math|Float|Lerp (Variables|Default|GetEdgeScale) 1.0 (Math|Float|Clamp(Float) (Utilities|Operators|Add (* (Utilities|Struct|BreakLinearColor _interpolatedcolor) _r) (* (.G (Utilities|Struct|BreakLinearColor _interpolatedcolor)) _g) (* _b (.B (Utilities|Struct|BreakLinearColor _interpolatedcolor))) (* (.A (Utilities|Struct|BreakLinearColor _interpolatedcolor)) _a))))) (Utilities|Operators|Multiply (Math|Random|RandomFloatinRangefromStream (Variables|Default|GetRandomStream) (Variables|Default|GetScaleMin) (Variables|Default|GetScaleMax)) (Math|Float|Lerp (Variables|Default|GetNoiseScale_Max) (Variables|Default|GetNoiseScale_Min) (Math|Float|Power (Math|Float|Clamp(Float) (/ (Math|Vector|Distance(Vector) (.location _array_element_2) (+ (Math|Vector|MakeVector _returnvalue (Math|Conversions|ToFloat(Integer) (Math|Float|Floor (.y (/ (.location _array_element_2) _cellnoisefrequency)))) (Math|Conversions|ToFloat(Integer) (Math|Float|Floor (.z (/ (.location _array_element_2) _cellnoisefrequency))))) "0.500000,0.500000,0.500000")) _cellnoisefrequency)) (Variables|Default|GetNoise_FallOff))) (Math|Float|Lerp 1.0 (/ (Utilities|Array|Get(acopy) _sampleradii _array_index) (Variables|Default|GetSampling%20Radius_Min)) (Variables|Default|GetScale_BySampleRadiusBlend)) (Math|Float|Lerp (Variables|Default|GetEdgeScale) 1.0 (Math|Float|Clamp(Float) (Utilities|Operators|Add (* (Utilities|Struct|BreakLinearColor _interpolatedcolor) _r) (* (.G (Utilities|Struct|BreakLinearColor _interpolatedcolor)) _g) (* _b (.B (Utilities|Struct|BreakLinearColor _interpolatedcolor))) (* (.A (Utilities|Struct|BreakLinearColor _interpolatedcolor)) _a))))) (Utilities|Operators|Multiply (Math|Random|RandomFloatinRangefromStream (Variables|Default|GetRandomStream) (Variables|Default|GetScaleMin) (Variables|Default|GetScaleMax)) (Math|Float|Lerp (Variables|Default|GetNoiseScale_Max) (Variables|Default|GetNoiseScale_Min) (Math|Float|Power (Math|Float|Clamp(Float) (/ (Math|Vector|Distance(Vector) (.location _array_element_2) (+ (Math|Vector|MakeVector _returnvalue (Math|Conversions|ToFloat(Integer) (Math|Float|Floor (.y (/ (.location _array_element_2) _cellnoisefrequency)))) (Math|Conversions|ToFloat(Integer) (Math|Float|Floor (.z (/ (.location _array_element_2) _cellnoisefrequency))))) "0.500000,0.500000,0.500000")) _cellnoisefrequency)) (Variables|Default|GetNoise_FallOff))) (Math|Float|Lerp 1.0 (/ (Utilities|Array|Get(acopy) _sampleradii _array_index) (Variables|Default|GetSampling%20Radius_Min)) (Variables|Default|GetScale_BySampleRadiusBlend)) (Math|Float|Lerp (Variables|Default|GetEdgeScale) 1.0 (Math|Float|Clamp(Float) (Utilities|Operators|Add (* (Utilities|Struct|BreakLinearColor _interpolatedcolor) _r) (* (.G (Utilities|Struct|BreakLinearColor _interpolatedcolor)) _g) (* _b (.B (Utilities|Struct|BreakLinearColor _interpolatedcolor))) (* (.A (Utilities|Struct|BreakLinearColor _interpolatedcolor)) _a))))))))
                (GeometryScript|VertexColor|SetMeshConstantVertexColor _tempmoss (Utilities|Struct|MakeLinearColor (Math|Float|Clamp(Float) (/ (Math|Vector|Distance(Vector) (.location _array_element_2) (+ (Math|Vector|MakeVector _returnvalue (Math|Conversions|ToFloat(Integer) (Math|Float|Floor (.y (/ (.location _array_element_2) _cellnoisefrequency)))) (Math|Conversions|ToFloat(Integer) (Math|Float|Floor (.z (/ (.location _array_element_2) _cellnoisefrequency))))) "0.500000,0.500000,0.500000")) _cellnoisefrequency))) (Utilities|Struct|MakeGeometryScriptColorFlags true) true)
                (bind _normallist (GeometryScript|Normals|GetMeshPerVertexNormals (Variables|Default|GettempMoss)))
                (for _index (range (+ (- (GeometryScript|ListUtils|GetVectorListLength _normallist) 0) 1))
                  (GeometryScript|ListUtils|SetVectorListItem _normallist _index (Math|Vector|GetUpVector (.rotation _array_element_2))))
                (GeometryScript|Normals|SetMeshToPerVertexNormals(FromList) (Variables|Default|GettempMoss) _normallist)
                (GeometryScript|MeshEdits|AppendMesh (Variables|Default|GetTarget%20Mesh) (Variables|Default|GettempMoss) (Math|Transform|MakeTransform "0, 0, 0") false (Utilities|Struct|MakeGeometryScriptAppendMeshOptions "UseSource"))))))))))

(fn Bake
  (bind _mossmesh (Variables|Default|GetMossMesh))
  (bind _output (Utilities|Array|Get(acopy) _mossmesh))
  (bind _returnvalue (StaticMesh|GetMaterial _output 0))
  (bind _meshnanitesettings (Utilities|Struct|MakeMeshNaniteSettings true))
  (bind _asset_path_and_name (Variables|Default|GetAsset%20Path%20and%20Name))
  (bind _returnvalue_2 (EditorScripting|Asset|GetPathNameforLoadedAsset (Variables|Default|GetStaticmesh) :!class /Script/EditorScriptingUtilities.EditorAssetLibrary))
  (Variables|Default|SetAsset%20Path%20and%20Name (Utilities|String|Append _returnvalue_2 "_MossAddon"))
  (bind _returnvalue_3 (EditorScripting|Asset|DoesAssetExist _asset_path_and_name :!class /Script/EditorScriptingUtilities.EditorAssetLibrary))
  (if _returnvalue_3
    (bind _returnvalue_4 (EditorScripting|Asset|LoadAsset _asset_path_and_name :!class /Script/EditorScriptingUtilities.EditorAssetLibrary))
    (bind _asstatic_mesh (Utilities|Casting|CastToStaticMesh _returnvalue_4))
    (GeometryScript|StaticMesh|CopyMeshtoStaticMesh (Variables|Default|GetTarget%20Mesh) _asstatic_mesh (Utilities|Struct|MakeGeometryScriptCopyMeshtoAssetOptions :bEnableRecomputeNormals false :bEnableRecomputeTangents true :bEnableRemoveDegenerates false :BoneHierarchyMismatchHandling "DoNothing" :bRemapBoneIndicesToMatchAsset false :bUseOriginalVertexOrder false :bUseBuildScale true :bReplaceMaterials false :NewMaterialUsage "ReplaceAssetMaterials" :bCleanAssignedMaterials false :GenerateLightmapUVs "MatchTargetLODSetting" :bApplyNaniteSettings true :NaniteSettings "(bEnabled=True,FallbackPercentTriangles=100.000000,FallbackRelativeError=0.000000)" :NewNaniteSettings _meshnanitesettings :bEmitTransaction true :bDeferMeshPostEditChange false)
      (:Success
        (StaticMesh|SetMaterial _asstatic_mesh 0 _returnvalue)))
    (else
      (bind _returnvalue_5 (GeometryScript|AssetManagement|CreateNewStaticMeshAssetfromMesh (Variables|Default|GetTarget%20Mesh) (Variables|Default|GetAsset%20Path%20and%20Name) (Utilities|Struct|MakeGeometryScriptCreateNewStaticMeshAssetOptions false false true _meshnanitesettings 0.0 false))
        (:Success
          (StaticMesh|SetMaterial _returnvalue_5 0 _returnvalue))))))

(fn ConstructionScript
  (bind _mesh (Variables|Default|GetMesh))
  (bind _staticmesh (Variables|Default|GetStaticmesh))
  (Components|StaticMesh|SetStaticMesh _mesh _staticmesh)
  (Rendering|Material|SetMaterial _mesh 0 (StaticMesh|GetMaterial _staticmesh 0))
  (Rendering|Material|SetMaterial (Variables|DynamicMeshActor|GetDynamicMeshComponent) 0 (StaticMesh|GetMaterial (Utilities|Array|Get(acopy) (Variables|Default|GetMossMesh)) 0)))
