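; Editor GeneratedDynamicMesh-style tool that gives a static mesh, by its name for moss, a near-zero-triangle Nanite fallback.
; On rebuild it copies the source's best available LOD into the target mesh and builds a separate fallback of two tiny discs at the bounding box corners, spanning its extents.
; Bake overwrites the asset with Nanite enabled, writing the fallback to the regular LOD, then the full mesh to the hi-res source.
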(event EventBeginPlay)

(event Collision|EventActorBeginOverlap
  (params (OtherActor :type Actor%20Object%20Reference)))

(event EventTick
  (params (DeltaSeconds :type Float%20%28single-precision%29)))

(event Events|EventOnRebuildGeneratedMesh
  (params (TargetMesh :type Dynamic%20Mesh%20Object%20Reference))
  (bind _staticmesh (Variables|Default|GetStaticmesh))
  (bind _returnvalue (StaticMesh|GetBoundingBox _staticmesh))
  (bind (_min _max) (Utilities|Struct|BreakBox _returnvalue))
  (bind _geometryscriptprimitiveoptions (Utilities|Struct|MakeGeometryScriptPrimitiveOptions "PerFace"))
  (Variables|Default|SetTarget%20Mesh TargetMesh)
  (bind _returnvalue_2 (DynamicMeshActor|AllocateComputeMesh))
  (bind _output_get (Variables|Default|SetFallbackMesh _returnvalue_2))
  (bind _returnvalue_3 (GeometryScript|Primitives|AppendDisc _output_get _geometryscriptprimitiveoptions (Math|Transform|MakeTransform _min) 0.1 2))
  (GeometryScript|Primitives|AppendDisc _returnvalue_3 _geometryscriptprimitiveoptions (Math|Transform|MakeTransform _max) 0.1 2)
  (GeometryScript|StaticMesh|CopyMeshFromStaticMeshwithSectionMaterials (Variables|Default|GetStaticmesh) (Variables|Default|GetTarget%20Mesh) (Utilities|Struct|MakeGeometryScriptCopyMeshfromAssetOptions true) (Utilities|Struct|MakeGeometryScriptMeshReadLOD "MaxAvailable")))

(fn Bake
  (bind _meshnanitesettings (Utilities|Struct|MakeMeshNaniteSettings true))
  (bind _geometryscriptcopymeshtoassetoptions (Utilities|Struct|MakeGeometryScriptCopyMeshtoAssetOptions :bEnableRecomputeNormals false :bEnableRecomputeTangents false :bEnableRemoveDegenerates false :BoneHierarchyMismatchHandling "DoNothing" :bRemapBoneIndicesToMatchAsset false :bUseOriginalVertexOrder false :bUseBuildScale true :bReplaceMaterials false :NewMaterialUsage "ReplaceAssetMaterials" :bCleanAssignedMaterials false :GenerateLightmapUVs "MatchTargetLODSetting" :bApplyNaniteSettings true :NaniteSettings "(bEnabled=True,FallbackPercentTriangles=100.000000,FallbackRelativeError=0.000000)" :NewNaniteSettings _meshnanitesettings :bEmitTransaction true :bDeferMeshPostEditChange false))
  (bind _asset_path_and_name (Variables|Default|GetAsset%20Path%20and%20Name))
  (bind _returnvalue (EditorScripting|Asset|GetPathNameforLoadedAsset (Variables|Default|GetStaticmesh) :!class /Script/EditorScriptingUtilities.EditorAssetLibrary))
  (Variables|Default|SetAsset%20Path%20and%20Name _returnvalue)
  (bind _returnvalue_2 (EditorScripting|Asset|DoesAssetExist _asset_path_and_name :!class /Script/EditorScriptingUtilities.EditorAssetLibrary))
  (if _returnvalue_2
    (bind _returnvalue_3 (EditorScripting|Asset|LoadAsset _asset_path_and_name :!class /Script/EditorScriptingUtilities.EditorAssetLibrary))
    (bind _asstatic_mesh (Utilities|Casting|CastToStaticMesh _returnvalue_3))
    (GeometryScript|StaticMesh|CopyMeshtoStaticMesh (Variables|Default|GetFallbackMesh) _asstatic_mesh _geometryscriptcopymeshtoassetoptions (Utilities|Struct|MakeGeometryScriptMeshWriteLOD false)
      (:Success
        (GeometryScript|StaticMesh|CopyMeshtoStaticMesh (Variables|Default|GetTarget%20Mesh) _asstatic_mesh _geometryscriptcopymeshtoassetoptions (Utilities|Struct|MakeGeometryScriptMeshWriteLOD true))))))

(fn ConstructionScript)
