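# Test moss material, masked, two-sided and Nanite-enabled, using Default Lit despite its SSS-named parameters.
# Color blends four tints by remapped vertex-color red and Fresnel, scaled by brightness; a moss texture times Opacity forms a mask value.
# Geometry math includes SplineThicken with TopWidth, plus a camera-cross-normal offset scaled by UV V and faded as normals face the camera.
# No material output connections appear, and several nodes stand unconnected.
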Material /Game/ArtTools/Procedural/MossGenerator/M_TestMoss
graph {
 properties {bUsedWithNanite=true, blendMode=BLEND_Masked, pixelDepthOffsetMode=PDOM_Legacy, twoSided=true, usedShadingModels=Default Lit}
 nodes [
  CameraVectorWS()
  Constant3Vector() {constant=<rgba 0 0 1 0>}
  n1 = VertexNormalWS()
  Fresnel(Normal=@n1)
  n2 = MaterialFunctionCall() {materialFunction=&CameraDirectionVector}
  n3 = CrossProduct(A=@n2.Result, B=@n1)
  n4 = ScalarParameter() {defaultValue=-18.9128, parameterName=Offset}
  n5 = TextureCoordinate()
  LinearInterpolate(
   A=Constant3Vector()
   B=Multiply(A=Multiply(A=@n3, B=@n4), B=ComponentMask(None=@n5) {g=true})
   Alpha=Abs(None=DotProduct(A=@n2.Result, B=@n1))
  )
  MaterialFunctionCall(WidthBase=Constant(), WidthTip=ScalarParameter() {defaultValue=5.5486, parameterName=TopWidth}) {materialFunction=&SplineThicken}
  Multiply(A=Constant3Vector() {constant=<rgba 0 0 1 0>}, B=TwoSidedSign())
  Multiply(
   A=TextureSample() {texture=&/MossTestTex}.R
   B=ScalarParameter() {defaultValue=1, parameterName=Opacity}
  )
  n6 = VectorParameter() {parameterName=Color, defaultValue=<rgba 0.0233 0.0781 0.013 1>}
  n7 = VectorParameter() {parameterName=Color3, defaultValue=<rgba 0.0649 0.0938 0.0211 1>}
  n8 = Saturate(None=LinearInterpolate(
    A=ScalarParameter() {defaultValue=-0.3914, parameterName=LowC}
    B=ScalarParameter() {defaultValue=5.3085, parameterName=HighC}
    Alpha=VertexColor().R
   ))
  n9 = VectorParameter() {parameterName=Color2, defaultValue=<rgba 0.1533 0.2917 0 1>}
  n10 = VectorParameter() {parameterName=Color4, defaultValue=<rgba 0.3073 0.3008 0.0704 1>}
  n11 = ScalarParameter() {defaultValue=3, parameterName=Fresnel}
  n12 = VertexNormalWS()
  Multiply(
   A=LinearInterpolate(
    A=LinearInterpolate(A=@n6.RGB, B=@n7.RGB, Alpha=@n8)
    B=LinearInterpolate(A=@n9.RGB, B=@n10.RGB, Alpha=@n8)
    Alpha=Fresnel(ExponentIn=@n11, Normal=@n12) {baseReflectFraction=0}
   )
   B=ScalarParameter() {defaultValue=0.5, parameterName=SSSBrightness}
  )
  ScalarParameter() {defaultValue=0.8757, parameterName=Roughnes}
  ScalarParameter() {defaultValue=0.5, parameterName=SSSOpacity}
  ScalarParameter() {parameterName=Spec}
  TextureSample() {texture=/Engine/EngineVolumetrics/Fogsheet/Materials/T_EV_FogSheet_Falloff_01}
 ]
}
