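# M_TestRock is a two-sided, Default Lit test rock material in the procedural MossGenerator tools folder, flagged for use with Nanite.
# It blends two vector parameters, a grey Color and a dark green Color2, using vertex color alpha as the mask, so per-vertex data controls where the green appears.
# A constant of one is also present, though the file does not show which material inputs either node drives.
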Material /Game/ArtTools/Procedural/MossGenerator/M_TestRock
graph {
 properties {bUsedWithNanite=true, pixelDepthOffsetMode=PDOM_Legacy, twoSided=true, usedShadingModels=Default Lit}
 nodes [
  Constant() {r=1}
  LinearInterpolate(
   A=VectorParameter() {parameterName=Color, defaultValue=<rgba 0.1198 0.1198 0.1198 1>}.RGB
   B=VectorParameter() {parameterName=Color2, defaultValue=<rgba 0.0124 0.0417 0.0069 1>}.RGB
   Alpha=VertexColor().A
  )
 ]
}
